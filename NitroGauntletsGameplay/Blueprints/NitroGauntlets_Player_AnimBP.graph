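# Player animation blueprint for the Nitro Gauntlets item that fills Item Override layers on the Fortnite player skeleton.
# ItemUpperBody swaps in a slide pose plus additive loop while sliding, otherwise passing its input through.
# ItemFinalPoseOverride applies meteor pitch to Root and a partial upper-body twist at the pelvis; while grinding, a lower-body blend profile mixes the unmodified pose back in.
# Other layers and the event graph have no logic.
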AnimBlueprint /NitroGauntletsGameplay/Blueprints/NitroGauntlets_Player_AnimBP
target_skeleton=&Fortnite_M_Avg_Player_Skeleton
graphs [
 AnimGraph {nodes [Output Pose()]}
 EventGraph {links [], nodes []}
 ItemAimOffset {nodes [
   Input Pose!() -> (Pose, AimOffsetAlpha:double, Yaw:double, Pitch:double) {name=InPose_AimOffset}
   Input Pose!() -> (Pose) {name=InPose_PreAimOffsets}
   Output Pose() {layerGroup=Item Override}
  ]}
 ItemFallAdditive {nodes [
   Input Pose() {name=InPoseItemFallAdditive}
   Output Pose() {layerGroup=Item Override}
  ]}
 ItemFallLandAdditive {nodes [
   Input Pose() {name=InPoseItemFallLandAdditive}
   Output Pose() {layerGroup=Item Override}
  ]}
 ItemFallLevel2Override {nodes [
   Input Pose() {name=InPose_FallLevel2}
   Output Pose() {layerGroup=Item Override}
  ]}
 ItemFinalPoseOverride {nodes [
   n1 = Use cached pose 'FinalPoseCache'()
   n2 = Use cached pose 'Post Melee Twist'()
   Output Pose(Result=Blend Poses by bool(
     BlendPose_0=Profile Blend!(BasePose:Pose Link Structure=@n2, BlendPose:Pose Link Structure=@n1, BlendWeight:float=1) -> (Pose:Pose Link Structure) {
      is=title AnimGraphNode_BlendProfileLayeredBlend
      blendProfileAsset=/Game/Characters/Player/BlendMasks/LowerBodyFeatheredBlendToAiming_BlendMask
      blendPose {linkId=16842751}
     }
     BlendPose_1=Use cached pose 'Post Melee Twist'()
     bActiveValue=Get IsGrinding() -> (IsGrinding:bool)
    )) {layerGroup=Item Override}
   Save cached pose 'FinalPoseCache'(Pose=Input Pose() {name=InFinalPose}) {cacheName=FinalPoseCache}
   n3 = "Transform (Modify) Bone - Bone: Root"!(
    ComponentPose=Local To Component(LocalPose=Use cached pose 'FinalPoseCache'())
    Alpha=Get MeteorPitchAlpha() -> (MeteorPitchAlpha:float)
    Translation="0,0,0"
    Rotation_Roll:float=Get MeteorPitchTotal() -> (MeteorPitchTotal:float)
    Rotation_Pitch:float=0
    Rotation_Yaw:float=0
    Scale="1,1,1"
   ) -> (Pose) {rotationMode=BMM_Replace, rotationSpace=BCS_BoneSpace, alphaBoolBlend {blendOption=Linear}, boneToModify {boneName=Root}}
   n4 = "Transform (Modify) Bone - Bone: pelvis"!(ComponentPose=@n3, Alpha=0.3, Rotation=Get UpperBodyTwistRotator() -> (UpperBodyTwistRotator:Rotator)) -> (Pose) {
    rotationMode=BMM_Additive
    rotationSpace=BCS_WorldSpace
    translationSpace=BCS_WorldSpace
    alphaBoolBlend {blendOption=Linear}
    boneToModify {boneName=pelvis}
   }
   Save cached pose 'Post Melee Twist'(Pose=Component To Local(ComponentPose=@n4)) {cacheName=Post Melee Twist}
  ]}
 ItemFullBodyAimOffsets {nodes [
   Input Pose!() -> (Pose, FullBody_AimOffsetAlpha:double, Aim_Yaw:double, Aim_Pitch:double) {name=InPoseFullBodyAimOffsets}
   Input Pose!() -> (Pose) {name=InPosePreFullBodyAimOffsets}
   Output Pose() {layerGroup=Item Override}
  ]}
 ItemFullBodyOverride {nodes [
   Input Pose() {name=InPoseFullBody}
   Output Pose() {layerGroup=Item Override}
  ]}
 ItemIdleOverride {nodes [
   Input Pose() {name=InPose_Idle}
   Output Pose() {layerGroup=Item Override}
  ]}
 ItemJetPackJumpAdditive {nodes [Output Pose() {layerGroup=Item Override}]}
 ItemJetPackStartAdditive {nodes [Output Pose() {layerGroup=Item Override}]}
 ItemJumpLoopAdditive {nodes [
   Input Pose() {name=InPoseItemJumpLoopAdditive}
   Output Pose() {layerGroup=Item Override}
  ]}
 ItemJumpOverride {nodes [
   Input Pose() {name=InPoseJump}
   Output Pose() {layerGroup=Item Override}
  ]}
 ItemJumpUpAdditive {nodes [
   Input Pose() {name=InPoseItemJumpUpAdditive}
   Output Pose() {layerGroup=Item Override}
  ]}
 ItemPreAimOffsets {nodes [
   Input Pose() {name=InPosePreAimOffsets}
   Output Pose() {layerGroup=Item Override}
  ]}
 ItemPreIK {nodes [
   Input Pose() {name=InPose_PreIK}
   Output Pose() {layerGroup=Item Override}
  ]}
 ItemRightHandIK {nodes [
   Input Pose() {name=InPosePreRightHandIK}
   Input Pose() {name=InPosePostRightHandIK}
   Output Pose() {layerGroup=Item Override}
  ]}
 ItemSkeletalControl {nodes [
   Input Pose()
   Output Pose() {layerGroup=Item Override}
  ]}
 ItemSwimJumpFallLoop {nodes [Output Pose() {layerGroup=Item Override}]}
 ItemSwimJumpSurfaceEnd {nodes [Output Pose() {layerGroup=Item Override}]}
 ItemSwimJumpSurfaceLoop {nodes [Output Pose() {layerGroup=Item Override}]}
 ItemUpperBody {nodes [
   n1 = Sequence Player 'NitroGauntlet_Slide_Pose1'() {bLoopAnimation=false, sequence=/NitroGauntletsGameplay/Assets/Anim/Sliding/NitroGauntlet_Slide_Pose1}
   n2 = "Sequence Player (additive) 'NitroGauntlet_Slide_Loop1'"() {bLoopAnimation=false, sequence=/NitroGauntletsGameplay/Assets/Anim/Sliding/NitroGauntlet_Slide_Loop1}
   Output Pose(Result=Blend Poses by bool(
     BlendPose_0=Apply Additive(Base=@n1, Additive=@n2) {alphaBoolBlend {blendOption=Linear}}
     BlendPose_1=Use cached pose 'UpperBodyCache'()
     bActiveValue=Get bIsSliding()
    ) {blendType=Linear}) {layerGroup=Item Override}
   Save cached pose 'UpperBodyCache'(Pose=Input Pose!() -> (Pose, InputParam:Fort Anim Input Adjusted Aim Structure) {name=InPoseUpperBody}) {cacheName=UpperBodyCache}
  ]}
 ItemZipLineStartAdditive {nodes [Output Pose() {layerGroup=Item Override}]}
]
variables [IsGrinding {type=boolean}]
