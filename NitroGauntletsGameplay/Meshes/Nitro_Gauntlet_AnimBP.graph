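# Animation Blueprint for the Nitro Gauntlets mesh, which follows the pose of the skeletal mesh it is attached to.
# A copied parent pose runs through the gauntlet Control Rig to form the base pose.
# A second copy feeds DefaultSlot, which is layered onto both nitro lower-arm bone branches for the output.
# The event graph is empty; additive, launch-sequence and chest-glow material nodes are left unconnected.
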AnimBlueprint /NitroGauntletsGameplay/Meshes/Nitro_Gauntlet_AnimBP
target_skeleton=&/Nitro_Gauntlet_Skeleton
graphs [
 AnimGraph {nodes [
   Apply Additive() {alphaBoolBlend {blendOption=Linear}}
   Additive Identity Pose()
   n1 = Control Rig(Source=Copy Pose From Mesh() {bUseAttachedParent=true}) {
    alphaBoolBlend {blendOption=Linear}
    controlRigAssetReference {blueprintRigClass=&/Nitro_Gauntlet_CtrlRig.Nitro_Gauntlet_CtrlRig_C}
   }
   n2 = Copy Pose From Mesh() {bUseAttachedParent=true}
   Output Pose(Result=Layered blend per bone(BasePose=@n1, BlendPoses_0=Slot 'DefaultSlot'(Source=@n2)) {layerSetup [{branchFilters [{boneName=nitro_lowerarm_l}, {boneName=nitro_lowerarm_r}]}]})
   "Sequence Evaluator (by time) 'NitroGauntlet_StartLaunchFromGround_WEP'"!(ExplicitTime:float=0) -> (Pose:Pose Link Structure) {is=title AnimGraphNode_SequenceEvaluator, sequence=/NitroGauntletsGameplay/Assets/Anim/WEP/NitroGauntlet_StartLaunchFromGround_WEP}
   n3 = GetPartSkeletalMeshComponent()
   SetVectorParameterValueOnMaterials!(
    execute
    self:Mesh Component Object Reference=@n3
    ParameterName:Name=GK_ChestGlowLocation
    ParameterValue:Vector=GetSocketLocation!(self:Scene Component Object Reference=@n3, InSocketName:Name=ChestGlowSocket) -> (ReturnValue:Vector)
   ) -> (then)
   Is Valid(InputObject=@n3)
  ]}
 EventGraph {links [], nodes []}
 NewFunction {nodes [NewFunction() {is=title K2Node_FunctionEntry}]}
]
variables [
 As B Weap Moonflax Nitro Gauntlets {type=B_Weap_Moonflax_NitroGauntlets_C, default=null}
 As Fort Player Pawn {type=FortPlayerPawn}
 L_Hand_Trans {type=Transform, default {location=<0 0 0>, rotation=<rot 0 0 0>, scale=<1 1 1>}}
]
